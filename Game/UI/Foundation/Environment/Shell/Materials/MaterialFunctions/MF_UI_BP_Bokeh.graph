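# UI material function that generates an animated, tinted bokeh layer, defaulting to a random-dots texture.
# Rotated coordinates feed foreground and background samples, each scaled separately and panned at different rates over time for a layered drift.
# Their red channels are summed, raised to a power, then multiplied by color and opacity.
# An adjustable linear falloff on one coordinate's green component, apparently vertical UV, fades out the result.
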MaterialFunction /Game/UI/Foundation/Environment/Shell/Materials/MaterialFunctions/MF_UI_BP_Bokeh
graph {nodes [
  n1 = FunctionInput(Preview=Constant() {r=0.125}) {bUsePreviewValueAsDefault=true, inputName=Bokeh Opacity, inputType=FunctionInput_Scalar, sortPriority=3}
  n2 = FunctionInput(Preview=Constant() {r=0.5}) {bUsePreviewValueAsDefault=true, inputName=Bokeh Scale, inputType=FunctionInput_Scalar, sortPriority=5}
  n3 = Rotator(
   Coordinate=NamedRerouteUsage()
   Time=FunctionInput(Preview=Constant() {r=-2.5}) {bUsePreviewValueAsDefault=true, inputName=Bokeh Angle, inputType=FunctionInput_Scalar, sortPriority=8}
  )
  n4 = Multiply(
   A=FunctionInput(Preview=Constant() {r=4}) {bUsePreviewValueAsDefault=true, inputName=Foreground Scale, inputType=FunctionInput_Scalar, sortPriority=6}
   B=@n3
  )
  n5 = Constant2Vector() {g=0.125, r=0.35}
  n6 = Multiply(
   A=NamedRerouteUsage()
   B=FunctionInput(Preview=Constant() {r=0.2}) {bUsePreviewValueAsDefault=true, inputName=Bokeh Speed, inputType=FunctionInput_Scalar, sortPriority=9}
  )
  n7 = TextureSample(UVs=Multiply(A=@n2, B=Panner(Coordinate=@n4, Time=@n6, Speed=@n5)), Tex=NamedRerouteUsage()) {texture=/Game/Global/Textures/Default/Noise/T_LockDeviceRandomDots}
  n8 = Multiply(
   A=@n3
   B=FunctionInput(Preview=Constant() {r=8}) {bUsePreviewValueAsDefault=true, inputName=Background Scale, inputType=FunctionInput_Scalar, sortPriority=7}
  )
  n9 = Constant2Vector() {g=0.25, r=0.25}
  n10 = TextureSample(UVs=Multiply(A=@n2, B=Panner(Coordinate=@n8, Time=@n6, Speed=@n9)), Tex=NamedRerouteUsage()) {texture=&T_LockDeviceRandomDots}
  n11 = Constant() {r=2.5}
  n12 = Multiply(
   A=FunctionInput(Preview=Constant3Vector() {constant=<rgba 1 1 1 1>}) {bUsePreviewValueAsDefault=true, inputName=Bokeh Color, sortPriority=2}
   B=Power(
    Base=Add(A=@n7.R, B=@n10.R)
    Exp=FunctionInput(Preview=@n11) {bUsePreviewValueAsDefault=true, inputName=Bokeh Power, inputType=FunctionInput_Scalar, sortPriority=4}
   )
  )
  n13 = MaterialFunctionCall(Float2=NamedRerouteUsage()) {materialFunction=&BreakOutFloat2Components}
  n14 = FunctionInput(Preview=Constant() {r=0.5}) {bUsePreviewValueAsDefault=true, inputName=Falloff Intensity, inputType=FunctionInput_Scalar, sortPriority=10}
  n15 = Constant() {r=9}
  n16 = Saturate(None=Multiply(
    A=Subtract(A=@n13.G, B=@n14)
    B=FunctionInput(Preview=@n15) {bUsePreviewValueAsDefault=true, inputName=Falloff Edge, inputType=FunctionInput_Scalar, sortPriority=11}
   ))
  FunctionOutput(None=Multiply(A=Multiply(A=@n1, B=@n12), B=OneMinus(None=@n16)))
  NamedRerouteDeclaration(None=Time()) {name=Time Value, nodeColor=<rgba 0 0 0 1>}
  NamedRerouteDeclaration(None=FunctionInput(Preview=TextureCoordinate()) {bUsePreviewValueAsDefault=true, inputName=UV, inputType=FunctionInput_Vector2}) {name=UV, nodeColor=<rgba 0 0 0 1>}
  NamedRerouteDeclaration(None=FunctionInput(Preview=TextureObject() {texture=&T_LockDeviceRandomDots}) {bUsePreviewValueAsDefault=true, inputName=Bokeh Texture, inputType=FunctionInput_Texture2D, sortPriority=1}) {name=Bokeh Texture, nodeColor=<rgba 0 0 0 1>}
 ]}
