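# Material function for UI background shell environments that supplies floor surface values and a Background Emissive output.
# Color is raised to a Color Intensity power, and RadialGradientExponential builds a floor mask from radius and density.
# Static switches choose between blended floor color, specular and roughness or flat black, non-specular, fully rough defaults.
# Metallic passes through unmasked, while named reroutes share the color, gradient and IsFloor flag.
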MaterialFunction /Game/UI/Foundation/Environment/Shell/Materials/MaterialFunctions/MF_UI_BP_FloorSetup
graph {nodes [
  n1 = NamedRerouteUsage()
  n2 = NamedRerouteUsage()
  FunctionOutput(None=StaticSwitch(True=LinearInterpolate(B=@n1, Alpha=@n2), False=Constant(), Value=NamedRerouteUsage())) {outputName=Floor Color}
  FunctionOutput(None=FunctionInput(Preview=Constant()) {bUsePreviewValueAsDefault=true, inputName=Floor Metallic, inputType=FunctionInput_Scalar, sortPriority=4}) {outputName=Floor Metallic, sortPriority=1}
  n3 = FunctionInput(Preview=Constant()) {bUsePreviewValueAsDefault=true, inputName=Floor Specular, inputType=FunctionInput_Scalar, sortPriority=5}
  n4 = NamedRerouteUsage()
  FunctionOutput(None=StaticSwitch(True=LinearInterpolate(B=@n3, Alpha=@n4), False=Constant(), Value=NamedRerouteUsage())) {outputName=Floor Specular, sortPriority=2}
  n5 = FunctionInput(Preview=Constant() {r=1}) {bUsePreviewValueAsDefault=true, inputName=Floor Roughness, inputType=FunctionInput_Scalar, sortPriority=6}
  n6 = NamedRerouteUsage()
  FunctionOutput(None=StaticSwitch(True=LinearInterpolate(B=@n5, Alpha=@n6) {constA=1}, False=Constant() {r=1}, Value=NamedRerouteUsage())) {outputName=Floor Roughness, sortPriority=3}
  FunctionOutput(None=NamedRerouteUsage()) {outputName=Background Emissive, sortPriority=4}
  n7 = TextureCoordinate()
  n8 = FunctionInput(Preview=Constant() {r=0.35}) {bUsePreviewValueAsDefault=true, inputName=Floor Radius, inputType=FunctionInput_Scalar, sortPriority=2}
  n9 = FunctionInput(Preview=Constant() {r=4}) {bUsePreviewValueAsDefault=true, inputName=Floor Density, inputType=FunctionInput_Scalar, sortPriority=3}
  FunctionOutput(None=NamedRerouteDeclaration(None=MaterialFunctionCall(UVs=@n7, Radius=@n8, Density=@n9) {materialFunction=&RadialGradientExponential}.RadialGradientExponential) {name=Floor Gradient, nodeColor=<rgba 0 0 0 1>}) {outputName=Floor Gradient Mask, sortPriority=20}
  n10 = Constant() {r=1}
  NamedRerouteDeclaration(None=Power(
    Base=FunctionInput() {bUsePreviewValueAsDefault=true, inputName=Color}
    Exp=FunctionInput(Preview=@n10) {bUsePreviewValueAsDefault=true, inputName=Color Intensity, inputType=FunctionInput_Scalar, sortPriority=1}
   )) {name=Final Color, nodeColor=<rgba 0 0 0 1>}
  NamedRerouteDeclaration(None=FunctionInput(Preview=StaticBool()) {bUsePreviewValueAsDefault=true, inputName=IsFloor, inputType=FunctionInput_StaticBool, sortPriority=7}) {name=IsFloor, nodeColor=<rgba 0 0 0 1>}
 ]}
