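# Material function for the shell environment UI that outputs a radial gradient whose brightness is scaled by a time input, named as a pulsing gradient.
# Offset UVs feed a saturated, inverted distance from the 0.5 center, shaped by Gradient Power.
# The result is multiplied by Gradient Intensity and Time.
# No oscillation happens inside, so pulsing depends on the supplied Time; the raw Time default grows linearly.
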MaterialFunction /Game/UI/Foundation/Environment/Shell/Materials/MaterialFunctions/MF_UI_BP_PulsingRadialGradient
graph {nodes [
  n1 = Add(
   A=FunctionInput(Preview=TextureCoordinate()) {bUsePreviewValueAsDefault=true, inputName=UV, inputType=FunctionInput_Vector2}
   B=FunctionInput(Preview=Constant2Vector() {g=1}) {bUsePreviewValueAsDefault=true, inputName=Gradient Offset, inputType=FunctionInput_Vector2, sortPriority=3}
  )
  n2 = Constant() {r=0.5}
  n3 = OneMinus(None=Saturate(None=Distance(A=@n1, B=@n2)))
  n4 = FunctionInput(Preview=Constant() {r=2}) {bUsePreviewValueAsDefault=true, inputName=Gradient Power, inputType=FunctionInput_Scalar, sortPriority=1}
  n5 = FunctionInput(Preview=Constant() {r=1}) {bUsePreviewValueAsDefault=true, inputName=Gradient Intensity, inputType=FunctionInput_Scalar, sortPriority=2}
  n6 = FunctionInput(Preview=Time()) {bUsePreviewValueAsDefault=true, inputName=Time, inputType=FunctionInput_Scalar, sortPriority=5}
  FunctionOutput(None=Multiply(A=Power(Base=@n3, Exp=@n4), B=Multiply(A=@n5, B=@n6)))
 ]}
