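# Material function for UI shell environment materials that provides screen-aligned UVs corrected for screen aspect ratio.
# It divides visible width by clamped height and scales the X axis of ScreenAlignedUVs by that ratio to output Squared UV.
# It also outputs a centering offset, half the difference between one and the aspect scale, and the screen's center point in squared UV space.
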MaterialFunction /Game/UI/Foundation/Environment/Shell/Materials/MaterialFunctions/MF_UI_BP_UvSetup
graph {nodes [
  n1 = MaterialFunctionCall(Float2=MaterialFunctionCall() {materialFunction=&ScreenResolution}.Visible Resolution) {materialFunction=&BreakOutFloat2Components}
  n2 = Divide(A=@n1.R, B=Max(A=@n1.G))
  n3 = AppendVector(A=@n2, B=Constant() {r=1})
  FunctionOutput(None=Multiply(A=MaterialFunctionCall() {materialFunction=&ScreenAlignedUVs}."X 100%, Y 100%", B=@n3)) {outputName=Squared UV}
  FunctionOutput(None=Multiply(A=Subtract(B=@n3)) {constB=0.5}) {outputName=Squared UV Center Offset, sortPriority=1}
  FunctionOutput(None=AppendVector(A=Multiply(A=@n2) {constB=0.5}, B=Constant() {r=0.5})) {outputName=Center Point, sortPriority=2}
 ]}
