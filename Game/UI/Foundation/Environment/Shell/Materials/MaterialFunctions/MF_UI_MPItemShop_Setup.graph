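# Material function for the MP item shop UI shell environment that outputs floor color, metallic, specular, roughness, gradient mask and background emissive.
# An exponential radial gradient forms the floor mask, with density scaled by a reveal transition.
# Brightness-scaled color, the mask and an IsFloor parameter are shared through named reroutes; nested lerps shape most floor outputs, while metallic passes straight through.
# Shading path switches give mobile separate density and roughness.
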MaterialFunction /Game/UI/Foundation/Environment/Shell/Materials/MaterialFunctions/MF_UI_MPItemShop_Setup
graph {nodes [
  n1 = NamedRerouteUsage()
  n2 = NamedRerouteUsage()
  FunctionOutput(None=LinearInterpolate(A=Constant(), B=LinearInterpolate(B=@n1, Alpha=@n2), Alpha=NamedRerouteUsage())) {outputName=Floor Color}
  FunctionOutput(None=FunctionInput(Preview=ScalarParameter() {defaultValue=1, group=6 - Floor, parameterName=Floor Metalic}) {bUsePreviewValueAsDefault=true, inputName=Floor Metallic, inputType=FunctionInput_Scalar, sortPriority=4}) {outputName=Floor Metallic, sortPriority=1}
  n3 = FunctionInput(Preview=ScalarParameter() {defaultValue=1, group=6 - Floor, parameterName=Floor Spec}) {bUsePreviewValueAsDefault=true, inputName=Floor Specular, inputType=FunctionInput_Scalar, sortPriority=5}
  n4 = NamedRerouteUsage()
  FunctionOutput(None=LinearInterpolate(A=Constant(), B=LinearInterpolate(B=@n3, Alpha=@n4), Alpha=NamedRerouteUsage())) {outputName=Floor Specular, sortPriority=2}
  n5 = LinearInterpolate(
   B=FunctionInput(Preview=ScalarParameter() {defaultValue=0.25, group=6 - Floor, parameterName=Floor Roughness}) {bUsePreviewValueAsDefault=true, inputName=Floor Roughness, inputType=FunctionInput_Scalar, sortPriority=6}
   Alpha=NamedRerouteUsage()
  )
  n6 = ScalarParameter() {defaultValue=0.75, group=6 - Floor, parameterName=Floor Roughness Mobile}
  n7 = LinearInterpolate(A=@n6, B=@n6, Alpha=NamedRerouteUsage()) {constA=0.75}
  FunctionOutput(None=LinearInterpolate(A=Constant() {r=1}, B=ShadingPathSwitch(Default=@n5, Mobile=@n7), Alpha=NamedRerouteUsage())) {outputName=Floor Roughness, sortPriority=3}
  FunctionOutput(None=NamedRerouteUsage()) {outputName=Background Emissive, sortPriority=4}
  n8 = FunctionInput(Preview=ScalarParameter() {defaultValue=1, group=6 - Floor, parameterName=Floor Density}) {bUsePreviewValueAsDefault=true, inputName=Floor Density, inputType=FunctionInput_Scalar, sortPriority=3}
  n9 = ScalarParameter() {defaultValue=1, group=6 - Floor, parameterName=Floor Reveal Transition, sliderMax=1}
  n10 = MaterialFunctionCall(
   UVs=TextureCoordinate()
   Radius=FunctionInput(Preview=ScalarParameter() {defaultValue=0.25, group=6 - Floor, parameterName=Floor Radius}) {bUsePreviewValueAsDefault=true, inputName=Floor Radius, inputType=FunctionInput_Scalar, sortPriority=2}
   Density=ShadingPathSwitch(Default=LinearInterpolate(B=@n8, Alpha=@n9), Mobile=ScalarParameter() {defaultValue=0.25, parameterName=FloorDensityMobile})
  ) {materialFunction=&RadialGradientExponential}
  FunctionOutput(None=NamedRerouteDeclaration(None=Saturate(None=@n10.RadialGradientExponential)) {name=Floor Gradient, nodeColor=<rgba 0 0 0 1>}) {outputName=Floor Gradient Mask, sortPriority=20}
  n11 = ScalarParameter() {defaultValue=1, group=0 - Global, parameterName=Color Brightness}
  NamedRerouteDeclaration(None=Multiply(
    A=FunctionInput() {bUsePreviewValueAsDefault=true, inputName=Color}
    B=FunctionInput(Preview=@n11) {bUsePreviewValueAsDefault=true, inputName=Color Brightness, inputType=FunctionInput_Scalar, sortPriority=1}
   )) {name=Final Color, nodeColor=<rgba 0 0 0 1>}
  NamedRerouteDeclaration(None=ScalarParameter() {parameterName=IsFloor}) {name=IsFloor, nodeColor=<rgba 0 0 0 1>}
 ]}
